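# Unlit, two-sided translucent Niagara sprite material for a radial energy wave effect in the SureBamboo emote.
# All textures pan in radial coordinates, with most UVs distorted by a shared normal-map noise "Disturb" reroute.
# Colour blends a two-colour radial gradient with plasma coils sampled at three offsets for RGB fringing, plus noise, tinted by particle colour.
# Opacity multiplies scrolling masks, radial gradients forming a ring, cell noise, a sphere mask and particle alpha.
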Material /BRCosmetics/Effects/Fort_Effects/Emotes/SureBamboo/Materials/M_Emote_SureBamboo_EnergyWave
graph {
 properties {
  bAutomaticallySetUsageInEditor=false
  bUsedWithNiagaraSprites=true
  blendMode=BLEND_Translucent
  shadingModel=MSM_Unlit
  twoSided=true
  usedShadingModels=Unlit
 }
 nodes [
  n1 = Constant() {r=0.6}
  n2 = StaticBool() {value=true}
  n3 = LinearInterpolate(
   A=VectorParameter() {defaultValue=<rgba 0.2552 1 0.952 1>}.RGB
   B=VectorParameter() {parameterName=Param_1, defaultValue=<rgba 0.5103 1 0.6891 1>}.RGB
   Alpha=Multiply(A=MaterialFunctionCall(Radius=@n1, Invert Density=@n2) {materialFunction=&RadialGradientExponential}.RadialGradientExponential)
  )
  n4 = VectorParameter() {parameterName=ColorUVScale, channelNames {a="", b="", g=VScale, r=UScale}, defaultValue=<rgba 2 1 0 0>}
  n5 = ComponentMask(None=@n4) {g=true, r=true}
  n6 = Add(
   A=Multiply(
    A=MaterialFunctionCall() {materialFunction=&VectorToRadialValue}.Radial Coordinates
    B=@n5
   )
   B=ScalarParameter() {defaultValue=0.1, parameterName=ColorOffset}
  ) {constB=0.1}
  n7 = NamedRerouteUsage()
  n8 = AppendVector(A=ScalarParameter() {parameterName=Color_PanSpeedX}, B=ScalarParameter() {defaultValue=-0.3, parameterName=Color_PanSpeedY})
  n9 = TextureSampleParameter2D(UVs=Panner(Coordinate=Add(A=@n6, B=@n7), Speed=@n8) {speedY=-0.3}) {parameterName=ColorTex, samplerType=SAMPLERTYPE_LinearGrayscale, texture=/Game/Effects/Fort_Effects/Emotes/UltraLight/Texture/T_PlasmaCoils}
  n10 = Add(A=Multiply(
    A=MaterialFunctionCall() {materialFunction=&VectorToRadialValue}.Radial Coordinates
    B=@n5
   )) {constB=0}
  n11 = NamedRerouteUsage()
  n12 = TextureSampleParameter2D(UVs=Panner(Coordinate=Add(A=@n10, B=@n11), Speed=@n8) {speedY=-0.3}) {parameterName=ColorTex, samplerType=SAMPLERTYPE_LinearGrayscale, texture=&T_PlasmaCoils}
  n13 = Subtract(
   A=Multiply(
    A=MaterialFunctionCall() {materialFunction=&VectorToRadialValue}.Radial Coordinates
    B=@n5
   )
   B=ScalarParameter() {defaultValue=0.1, parameterName=ColorOffset}
  ) {constB=0}
  n14 = NamedRerouteUsage()
  n15 = TextureSampleParameter2D(UVs=Panner(Coordinate=Add(A=@n13, B=@n14), Speed=@n8) {speedY=-0.3}) {parameterName=ColorTex, samplerType=SAMPLERTYPE_LinearGrayscale, texture=&T_PlasmaCoils}
  n16 = Power(Base=MaterialFunctionCall(X=@n9.R, Y=@n12.R, Z=@n15.R) {materialFunction=&MakeFloat3}.Result) {constExponent=0.5}
  n17 = Constant() {r=-1}
  n18 = VectorParameter() {parameterName=Mask_TexB_UVScale, channelNames {a="", b="", g=VScale, r=UScale}, defaultValue=<rgba 2 0.3 0 0>}
  n19 = Multiply(
   A=MaterialFunctionCall() {materialFunction=&VectorToRadialValue}.Radial Coordinates
   B=ComponentMask(None=@n18) {g=true, r=true}
  )
  n20 = NamedRerouteUsage()
  n21 = ScalarParameter() {parameterName=TexB_PanSpeedX}
  n22 = ScalarParameter() {defaultValue=-0.2, parameterName=TexB_PanSpeedY}
  n23 = TextureSampleParameter2D(UVs=Panner(Coordinate=Add(A=@n19, B=@n20), Speed=AppendVector(A=@n21, B=@n22)) {speedY=-0.2}) {parameterName=Mask_TexB, texture=/Game/Effects/Fort_Effects/Textures/Tile/T_Noise_03}
  n24 = VectorParameter() {parameterName=Mask_TexA_UVScale, channelNames {a="", b="", g=VScale, r=UScale}, defaultValue=<rgba 2 1 0 0>}
  n25 = Multiply(
   A=MaterialFunctionCall() {materialFunction=&VectorToRadialValue}.Radial Coordinates
   B=ComponentMask(None=@n24) {g=true, r=true}
  )
  n26 = NamedRerouteUsage()
  n27 = ScalarParameter() {parameterName=TexA_PanSpeedX}
  n28 = ScalarParameter() {defaultValue=-0.3, parameterName=TexA_PanSpeedY}
  n29 = TextureSampleParameter2D(UVs=Panner(Coordinate=Add(A=@n25, B=@n26), Speed=AppendVector(A=@n27, B=@n28)) {speedY=-0.3}) {parameterName=Mask_TexA, samplerType=SAMPLERTYPE_LinearGrayscale, texture=&T_PlasmaCoils}
  n30 = Add(A=@n23.RGB, B=Multiply(A=Power(Base=@n29.R)) {constB=2})
  n31 = Add(A=Power(Base=Desaturation(None=@n16, Fraction=@n17)), B=Power(Base=LinearInterpolate(Alpha=@n30) {constB=3}) {constExponent=1})
  n32 = ParticleColor()
  MaterialFunctionCall(Emissive=Multiply(A=Multiply(A=@n3, B=@n31), B=@n32.RGB) {constA=1}) {materialFunction=&MF_FX_EmissiveAdjustments}
  n33 = Constant()
  n34 = SmoothStep(Min=@n33, Max=Add(A=@n33, B=Constant() {r=1}), Value=LinearInterpolate(A=@n29.R, B=@n23.R) {constAlpha=0.7})
  n35 = MaterialFunctionCall() {materialFunction=&RadialGradientExponential}
  n36 = ScalarParameter() {defaultValue=4, parameterName="Radius (S)"}
  n37 = ScalarParameter() {defaultValue=10, parameterName="Density (S)"}
  n38 = StaticBool() {value=true}
  n39 = OneMinus(None=Multiply(A=MaterialFunctionCall(Radius=@n36, Density=@n37, Invert Density=@n38) {materialFunction=&RadialGradientExponential}.RadialGradientExponential))
  n40 = Saturate(None=Multiply(A=Multiply(A=@n35.RadialGradientExponential, B=@n39)) {constB=2})
  n41 = Power(Base=Multiply(A=Multiply(A=@n34, B=@n40)) {constB=2}, Exp=ScalarParameter() {defaultValue=1, parameterName=MaskPower})
  n42 = Add(
   A=Multiply(
    A=MaterialFunctionCall() {materialFunction=&VectorToRadialValue}.Radial Coordinates
    B=Constant2Vector() {g=0.5, r=2}
   )
   B=NamedRerouteUsage()
  )
  n43 = Desaturation(None=TextureSample(UVs=Panner(Coordinate=@n42) {speedY=-0.3}) {samplerType=SAMPLERTYPE_Grayscale, texture=/BRCosmetics/Weapons/FORT_Melee/Pickaxe_Cube_Ninja_Male/FX/T_CellNoise}.RGB)
  n44 = Saturate(None=Multiply(A=LinearInterpolate(Alpha=@n43) {constA=0.5}) {constB=5})
  n45 = TextureCoordinate()
  n46 = Constant() {r=0.5}
  n47 = Power(Base=OneMinus(None=SphereMask(A=@n45, B=@n46) {attenuationRadius=0.5, hardnessPercent=0})) {constExponent=3}
  Multiply(A=Multiply(A=@n32.A, B=Multiply(A=@n41, B=@n44)), B=Saturate(None=Multiply(A=@n47) {constB=5}))
  n48 = MaterialFunctionCall() {materialFunction=&VectorToRadialValue}
  n49 = Constant2Vector() {g=0.1, r=4}
  n50 = TextureSample(UVs=Panner(Coordinate=Multiply(A=@n48.Radial Coordinates, B=@n49)) {speedX=0.1}) {samplerType=SAMPLERTYPE_Normal, texture=/Game/Characters/Player/Female/Medium/Bodies/F_MED_Bistro_Astronaut/FX/T_GaussianNoise_Normal}
  NamedRerouteDeclaration(None=Multiply(A=ComponentMask(None=@n50.RGB) {g=true, r=true}) {constB=0.2}) {name=Disturb, nodeColor=<rgba 0 1 0.1562 1>}
 ]
}
