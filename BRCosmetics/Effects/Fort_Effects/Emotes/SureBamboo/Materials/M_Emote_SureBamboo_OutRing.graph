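# Unlit, translucent, two-sided Niagara particle material for the outer ring effect of the SureBamboo emote.
# Nested radial exponential gradients blend near-black into teal and bright cyan, tinted by particle color and passed through MF_FX_EmissiveAdjustments.
# A gradient multiplied by another's inverse forms a soft ring mask, scaled by particle alpha.
# Fresnel and view-angle fade branches are also present, with no visible link to the ring.
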Material /BRCosmetics/Effects/Fort_Effects/Emotes/SureBamboo/Materials/M_Emote_SureBamboo_OutRing
graph {
 properties {
  bAutomaticallySetUsageInEditor=false
  bUsedWithNiagaraMeshParticles=true
  bUsedWithNiagaraSprites=true
  blendMode=BLEND_Translucent
  shadingModel=MSM_Unlit
  twoSided=true
  usedShadingModels=Unlit
 }
 nodes [
  n1 = Constant3Vector() {constant=<rgba 0.0003 0.0048 0.0104 1>}
  n2 = MaterialFunctionCall(Radius=Constant() {r=0.7}, Density=Constant() {r=5}, Invert Density=StaticBool()) {materialFunction=&RadialGradientExponential}
  n3 = LinearInterpolate(
   A=Constant3Vector() {constant=<rgba 0 0.2022 0.3 1>}
   B=Constant3Vector() {constant=<rgba 0 2 1.9926 1>}
   Alpha=OneMinus(None=Multiply(A=@n2.RadialGradientExponential))
  )
  n4 = Constant() {r=0.5}
  n5 = Constant() {r=1}
  n6 = StaticBool() {value=true}
  n7 = OneMinus(None=Multiply(A=MaterialFunctionCall(Radius=@n4, Density=@n5, Invert Density=@n6) {materialFunction=&RadialGradientExponential}.RadialGradientExponential))
  n8 = ParticleColor()
  MaterialFunctionCall(Emissive=Multiply(A=LinearInterpolate(A=@n1, B=@n3, Alpha=@n7), B=@n8.RGB)) {materialFunction=&MF_FX_EmissiveAdjustments}
  n9 = Constant() {r=0.55}
  n10 = Constant() {r=5}
  n11 = StaticBool()
  n12 = OneMinus(None=Multiply(A=MaterialFunctionCall(Radius=@n9, Density=@n10, Invert Density=@n11) {materialFunction=&RadialGradientExponential}.RadialGradientExponential))
  n13 = Multiply(A=MaterialFunctionCall(Radius=Constant() {r=0.5}, Density=Constant() {r=3}, Invert Density=StaticBool()) {materialFunction=&RadialGradientExponential}.RadialGradientExponential)
  Multiply(A=@n8.A, B=Multiply(A=Multiply(A=@n12, B=@n13)) {constB=1.5})
  OneMinus()
  Power(Base=Fresnel() {baseReflectFraction=0.05, exponent=2}) {constExponent=2.5}
  Power(Base=Multiply(A=Fresnel() {baseReflectFraction=0, exponent=2}) {constB=2}) {constExponent=1}
  n14 = MaterialFunctionCall("FadeSensitivity (optional)"=Constant() {r=3}) {materialFunction=&FadeBasedOnViewAngle}
  Saturate(None=Power(Base=Multiply(A=@n14.Result) {constB=5}))
  StaticBool() {value=true}
 ]
}
