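# Unlit, translucent, two-sided material for Niagara mesh particles that draws a teal, animated energy ring for the SureBamboo emote.
# UV1 becomes radial coordinates, panned and distorted by blurred noise to sample electric crystal noise for glow and plasma for alpha.
# A Fresnel gradient underlies the glow, sphere masks cut the hole and fade the rim, and particle color tints and fades it.
# Explicit UV derivatives drive mip selection in the radial samples.
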Material /BRCosmetics/Effects/Fort_Effects/Emotes/SureBamboo/Materials/M_Emote_SureBamboo_RingEnergy
graph {
 properties {
  bUsedWithNiagaraMeshParticles=true
  blendMode=BLEND_Translucent
  pixelDepthOffsetMode=PDOM_Legacy
  shadingModel=MSM_Unlit
  twoSided=true
  usedShadingModels=Unlit
 }
 nodes [
  n1 = Power(Base=Fresnel() {baseReflectFraction=0, exponent=1}) {constExponent=0.5}
  n2 = LinearInterpolate(
   A=Add(A=Constant3Vector() {constant=<rgba 0 0.3333 0.3167 1>}) {constB=0}
   B=Constant3Vector() {constant=<rgba 0 0.05 0.0458 1>}
   Alpha=Saturate(None=Multiply(A=@n1))
  )
  n3 = Constant3Vector() {constant=<rgba 0.1009 1 0.955 1>}
  n4 = TextureCoordinate() {coordinateIndex=1, uTiling=2, vTiling=2}
  n5 = Constant() {r=1}
  n6 = Multiply(
   A=Constant2Vector() {g=1, r=1}
   B=MaterialFunctionCall(Vector or UVs=Subtract(A=@n4, B=@n5)) {materialFunction=&VectorToRadialValue}.Radial Coordinates
  )
  n7 = MaterialFunctionCall(Vector or UVs=Subtract(A=TextureCoordinate() {coordinateIndex=1, uTiling=2, vTiling=2}, B=Constant() {r=1})) {materialFunction=&VectorToRadialValue}
  n8 = Constant2Vector() {g=1, r=1}
  n9 = TextureCoordinate() {coordinateIndex=1}
  n10 = TextureSample!(
   UVs=Panner(Coordinate=Multiply(A=@n7.Radial Coordinates, B=@n8)) {speedX=-0.2}
   Tex
   "DDX(UVs)"=DDX(Value=@n9)
   "DDY(UVs)"=DDY(Value=@n9)
   Apply View MipBias
  ) {
   mipValueMode=TMVM_Derivative
   samplerType=SAMPLERTYPE_LinearGrayscale
   texture=/Engine/Functions/Engine_MaterialFunctions02/ExampleContent/Textures/LowResBlurredNoise
  }
  n11 = Multiply(A=@n10.R) {constB=0.2}
  n12 = TextureCoordinate() {coordinateIndex=1}
  n13 = TextureSample(UVs=Panner(Coordinate=Add(A=@n6, B=@n11)) {speedX=0.2}, "DDX(UVs)"=DDX(Value=@n12), "DDY(UVs)"=DDY(Value=@n12)) {
   mipValueMode=TMVM_Derivative
   samplerType=SAMPLERTYPE_LinearGrayscale
   texture=/Game/Gadgets/Assets/VinderTech_GliderChute/Glider_Dazzle/FX/Variant/T_Glider_Dazzle_Electric_Crystal_Noise
  }
  n14 = Multiply(A=@n13.R) {constB=3}
  n15 = Add(A=Clamp(None=Power(Base=@n14) {constExponent=1}) {maxDefault=5}) {constB=0}
  n16 = SphereMask(A=TextureCoordinate() {coordinateIndex=1}, B=Constant() {r=0.5}, Radius=Constant() {r=0.5}) {hardnessPercent=50}
  n17 = Multiply(A=LinearInterpolate(B=Multiply(A=@n15)) {constAlpha=0.3}, B=LinearInterpolate(Alpha=OneMinus(None=@n16)) {constB=0.5})
  n18 = ParticleColor()
  n19 = Multiply(A=Add(A=Multiply(A=@n2), B=Multiply(A=@n3, B=@n17)) {constB=0}, B=@n18.RGB)
  n20 = TextureCoordinate() {coordinateIndex=1}
  n21 = Constant() {r=0.5}
  MaterialFunctionCall(Emissive=LinearInterpolate(A=Add(B=@n19), Alpha=SphereMask(A=@n20, B=@n21) {attenuationRadius=0.2, hardnessPercent=50}) {constB=0}) {materialFunction=&MF_FX_EmissiveAdjustments}
  n22 = TextureCoordinate() {coordinateIndex=1}
  n23 = Constant() {r=0.5}
  n24 = TextureCoordinate() {coordinateIndex=1, uTiling=2, vTiling=2}
  n25 = Constant() {r=1}
  n26 = Multiply(
   A=Constant2Vector() {g=1, r=1}
   B=MaterialFunctionCall(Vector or UVs=Subtract(A=@n24, B=@n25)) {materialFunction=&VectorToRadialValue}.Radial Coordinates
  )
  n27 = MaterialFunctionCall(Vector or UVs=Subtract(A=TextureCoordinate() {coordinateIndex=1, uTiling=2, vTiling=2}, B=Constant() {r=1})) {materialFunction=&VectorToRadialValue}
  n28 = Constant2Vector() {g=1, r=1}
  n29 = TextureCoordinate() {coordinateIndex=1}
  n30 = TextureSample(UVs=Panner(Coordinate=Multiply(A=@n27.Radial Coordinates, B=@n28)) {speedX=-0.1}, "DDX(UVs)"=DDX(Value=@n29), "DDY(UVs)"=DDY(Value=@n29)) {
   mipValueMode=TMVM_Derivative
   samplerType=SAMPLERTYPE_LinearGrayscale
   texture=&LowResBlurredNoise
  }
  n31 = Multiply(A=@n30.R) {constB=0.2}
  n32 = TextureCoordinate() {coordinateIndex=1}
  n33 = TextureSample(UVs=Panner(Coordinate=Add(A=@n26, B=@n31)) {speedX=0.2}, "DDX(UVs)"=DDX(Value=@n32), "DDY(UVs)"=DDY(Value=@n32)) {mipValueMode=TMVM_Derivative, samplerType=SAMPLERTYPE_LinearColor, texture=/Game/Effects/Fort_Effects/Emotes/UltraLight/Texture/T_Plasma_D_C}
  n34 = TextureCoordinate() {coordinateIndex=1}
  n35 = Constant() {r=0.5}
  n36 = Constant() {r=0.5}
  n37 = Power(Base=@n33.R, Exp=LinearInterpolate(Alpha=SphereMask(A=@n34, B=@n35, Radius=@n36) {hardnessPercent=0}) {constA=3, constB=0})
  n38 = Saturate(None=Multiply(A=Multiply(A=@n37) {constB=2}))
  n39 = TextureCoordinate() {coordinateIndex=1}
  n40 = Constant() {r=0.5}
  n41 = Constant() {r=0.5}
  Multiply(
   A=Multiply(A=Multiply(A=@n18.A, B=@n38), B=SphereMask(A=@n39, B=@n40, Radius=@n41) {hardnessPercent=0})
   B=OneMinus(None=SphereMask(A=@n22, B=@n23) {attenuationRadius=0.2, hardnessPercent=50})
  )
 ]
}
