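# Translucent, two-sided unlit Niagara sprite material for the SureBamboo pickaxe impact wave, drawing a radial spiked shockwave tinted by particle color.
# Radial coordinates map a spike texture, distorted by smoke noise, with the dynamic Panner parameter driving its outward/inward scroll time.
# Cell noise breaks up the spikes, and two exponential radial gradients mask them.
# The dynamic Dissolve parameter drives a noise-based HeightLerp erosion, before particle alpha and DepthFade finish opacity.
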Material /BRCosmetics/Effects/Fort_Effects/Emotes/SureBamboo/Materials/M_Pickaxe_SureBamboo_ImpactWave
graph {
 properties {
  bAutomaticallySetUsageInEditor=false
  bUsedWithNiagaraSprites=true
  blendMode=BLEND_Translucent
  shadingModel=MSM_Unlit
  twoSided=true
  usedShadingModels=Unlit
 }
 nodes [
  n1 = ParticleColor()
  n2 = Multiply(
   A=MaterialFunctionCall() {materialFunction=&VectorToRadialValue}.Radial Coordinates
   B=Constant2Vector() {g=1, r=2.5}
  )
  n3 = Multiply(
   A=MaterialFunctionCall() {materialFunction=&VectorToRadialValue}.Radial Coordinates
   B=Constant2Vector() {g=0.1, r=4}
  )
  n4 = Multiply(A=TextureSample(UVs=Panner(Coordinate=@n3)) {texture=&T_Inky_Smoke_Tile}.R) {constB=0.2}
  n5 = DynamicParameter() {defaultValue=<rgba 1 0 1 1>, paramNames [Panner,Dissolve,Param3,Param4]}
  n6 = TextureSample(UVs=Panner(Coordinate=Add(A=@n2, B=@n4) {constB=0}, Time=LinearInterpolate(Alpha=@n5.Panner) {constA=-0.2, constB=0.1}) {speedY=-2}) {samplerSource=SSM_Wrap_WorldGroupSettings, texture=/Game/Effects/Niagara/MasterMaterials/ShockWave/Texture/T_Wave_Spikes}
  n7 = ScalarParameter() {defaultValue=1, parameterName=Power}
  n8 = Add(A=Multiply(
    A=MaterialFunctionCall() {materialFunction=&VectorToRadialValue}.Radial Coordinates
    B=Constant2Vector() {r=2}
   )) {constB=0}
  n9 = LinearInterpolate(Alpha=TextureSample(UVs=Panner(Coordinate=@n8)) {samplerType=SAMPLERTYPE_Grayscale, texture=/BRCosmetics/Weapons/FORT_Melee/Pickaxe_Cube_Ninja_Male/FX/T_CellNoise}.R) {constA=-0.2, constB=2}
  n10 = Multiply(A=@n1.A, B=Multiply(A=Power(Base=@n6.R, Exp=@n7) {constExponent=1}, B=Saturate(None=@n9)))
  n11 = Multiply(A=MaterialFunctionCall(Density=Constant() {r=1}) {materialFunction=&RadialGradientExponential}.RadialGradientExponential)
  n12 = Multiply(A=MaterialFunctionCall(
    Radius=ScalarParameter() {defaultValue=10, parameterName="Radius (S)"}
    Density=ScalarParameter() {defaultValue=30, parameterName="Density (S)"}
    Invert Density=StaticBool() {value=true}
   ) {materialFunction=&RadialGradientExponential}.RadialGradientExponential)
  n13 = Multiply(A=Multiply(A=Power(Base=@n11), B=OneMinus(None=@n12))) {constB=2}
  n14 = Saturate(None=Multiply(A=Saturate(None=@n13)) {constB=5})
  n15 = DynamicParameter() {defaultValue=<rgba 1 0 1 1>, paramNames [Panner,Dissolve,Param3,Param4]}
  n16 = MaterialFunctionCall() {materialFunction=&VectorToRadialValue}
  n17 = Constant2Vector() {g=0.2, r=8}
  n18 = Panner(Coordinate=Add(A=Multiply(A=@n16.Radial Coordinates, B=@n17)) {constB=0})
  n19 = Saturate(None=Multiply(A=TextureSample(UVs=@n18) {samplerType=SAMPLERTYPE_Grayscale, texture=&T_CellNoise}.R) {constB=2})
  DepthFade(
   Opacity=Multiply(
    A=Multiply(A=@n10, B=@n14)
    B=MaterialFunctionCall(Transition Phase=@n15.Dissolve, Height Texture=@n19) {materialFunction=&HeightLerp}.Results
   )
   FadeDistance=ScalarParameter() {defaultValue=10, parameterName=FadeDistance}
  )
  MaterialFunctionCall(Emissive=@n1.RGB) {materialFunction=&MF_FX_EmissiveAdjustments}
 ]
}
